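; Editor-only waterfall generator that builds a dynamic mesh from widget-defined splines and bakes it to a static mesh and a WaterfallGeneration_Visual actor.
; On rebuild, sampled fall and extend splines become swept polygons, plus a splash disc, at three LODs, merged with per-part material IDs.
; Material dimensions drive preview instances, with debug material options.
; Delayed timers line-trace for splash points, then bake meshes, material data and FX parameters.
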(event Events|EventOnRebuildGeneratedMesh
  (params (TargetMesh :type Dynamic%20Mesh%20Object%20Reference))
  (if (Editor|InEditor)
    (Utilities|FlowControl|Sequence
      (:then_0
        (GSRefreshFunc|UserInterfaceLimit3DWidget))
      (:then_1
        (GSRefreshFunc|GenerationSplineGen))
      (:then_2
        (GSRefreshFunc|GenerationSampleSpline))
      (:then_5
        (GSRefreshFunc|GenerationMeshGen))
      (:then_6
        (GSRefreshFunc|GenerationMeshMerge))
      (:then_7
        (GSRefreshFunc|GenerationMaterialDataCollect)
        (GSRefreshFunc|GenerationPreviewMaterialSetting))
      (:then_9
        (Utilities|Time|SetTimerbyFunctionName (Variables|Default|GetTickHelper) "FXDetect_Fix" 0.2222)
        (Utilities|Time|SetTimerbyFunctionName (Variables|Default|GetTickHelper) "BakeWaterfallData" 0.3333))
      (:then_10
        (Development|PrintString "WaterfallGenerationTool:ExecuteRebuildMesh")))
    (else
      (Development|PrintString "WaterfallGenerationTool:WhyImInGame?"))))

(event EventBakeWaterfallData
  (GSRefreshFunc|BakeDatatoVisualActor))

(event EventFXDetect_Fix
  (Tool|DetectSplashPoint))

(fn BakeDataToVisualActor
  (bind _spline_bevelandfall (Variables|Default|GetSpline_BevelAndFall))
  (bind _returnvalue (Spline|GetLocationatSplinePoint _spline_bevelandfall 0 "World"))
  (bind _targetvisualbp (Variables|%5BUser%5DInitialSetting|GetTargetVisualBP))
  (bind _targetvisualbp_2 (Variables|%5BUser%5DInitialSetting|GetTargetVisualBP))
  (bind _targetvisualbp_3 (Variables|%5BUser%5DInitialSetting|GetTargetVisualBP))
  (bind _targetvisualbp_4 (Variables|%5BUser%5DInitialSetting|GetTargetVisualBP))
  (Utilities|FlowControl|Sequence
    (:then_0
      ;[1 1 1] BakeMesh
      (Utilities|FlowControl|Sequence
        (:then_0
          ;;[1 1 1] Bake to static mesh
          (Utilities|IsValid (Variables|%5BUser%5DInitialSetting|GetTargetStaticMesh)
            (:Is%20Valid
              (bind _returnvalue_2 (DynamicMeshComponent|GetDynamicMesh (Variables|DynamicMeshActor|GetDynamicMeshComponent)))
              (GeometryScript|StaticMesh|CopyMeshtoStaticMesh _returnvalue_2 (Variables|%5BUser%5DInitialSetting|GetTargetStaticMesh) (Utilities|Struct|MakeGeometryScriptCopyMeshtoAssetOptions false false false "DoNothing" false false true true "ReplaceAssetMaterials" (Utilities|Array|MakeArray "/Engine/EngineMaterials/WorldGridMaterial.WorldGridMaterial" "/Engine/EngineMaterials/WorldGridMaterial.WorldGridMaterial" "/Engine/EngineMaterials/WorldGridMaterial.WorldGridMaterial")) (Utilities|Struct|MakeGeometryScriptMeshWriteLOD false)))))
        (:then_1
          ;;[1 1 1] Bake to static mesh
          (Utilities|IsValid (Variables|%5BUser%5DInitialSetting|GetTargetStaticMesh)
            (:Is%20Valid
              (bind _returnvalue_3 (DynamicMeshComponent|GetDynamicMesh (Variables|Default|GetDynamicMesh_Merged_LOD1)))
              (GeometryScript|StaticMesh|CopyMeshtoStaticMesh :FromDynamicMesh _returnvalue_3 :ToStaticMeshAsset (Variables|%5BUser%5DInitialSetting|GetTargetStaticMesh) :TargetLOD (Utilities|Struct|MakeGeometryScriptMeshWriteLOD false 1) :bUseSectionMaterials true))))
        (:then_2
          ;;[1 1 1] Bake to static mesh
          (Utilities|IsValid (Variables|%5BUser%5DInitialSetting|GetTargetStaticMesh)
            (:Is%20Valid
              (bind _returnvalue_4 (DynamicMeshComponent|GetDynamicMesh (Variables|Default|GetDynamicMesh_Merged_LOD2)))
              ;;;[1 1 1] Comment
              (GeometryScript|StaticMesh|CopyMeshtoStaticMesh :FromDynamicMesh _returnvalue_4 :ToStaticMeshAsset (Variables|%5BUser%5DInitialSetting|GetTargetStaticMesh) :TargetLOD (Utilities|Struct|MakeGeometryScriptMeshWriteLOD false 2) :bUseSectionMaterials true))))))
    (:then_1
      ;[1 1 1] MaterialBasicParam
      (Variables|/Game/Tools/WaterfallGenerator/WaterfallGeneration_Visual.WaterfallGeneration_Visual_C|SetThickness (Variables|IgnoreThis%5BWorkVar%5D|GetThickness) _targetvisualbp_4)
      (Variables|/Game/Tools/WaterfallGenerator/WaterfallGeneration_Visual.WaterfallGeneration_Visual_C|SetWidth (Variables|IgnoreThis%5BWorkVar%5D|GetMaterialData_Width) _targetvisualbp_4)
      (Variables|/Game/Tools/WaterfallGenerator/WaterfallGeneration_Visual.WaterfallGeneration_Visual_C|SetHeight (Variables|IgnoreThis%5BWorkVar%5D|GetMaterialData_Height) _targetvisualbp_4)
      (Variables|/Game/Tools/WaterfallGenerator/WaterfallGeneration_Visual.WaterfallGeneration_Visual_C|SetEdgeOffset (Variables|IgnoreThis%5BWorkVar%5D|GetMaterialData_EdgeOffset) _targetvisualbp_4)
      (Variables|/Game/Tools/WaterfallGenerator/WaterfallGeneration_Visual.WaterfallGeneration_Visual_C|SetBevelLength (Variables|IgnoreThis%5BWorkVar%5D|GetMaterialData_BevelLength) _targetvisualbp_4)
      (Variables|/Game/Tools/WaterfallGenerator/WaterfallGeneration_Visual.WaterfallGeneration_Visual_C|SetExtendLength (Variables|IgnoreThis%5BWorkVar%5D|GetMaterialData_ExtendLength) _targetvisualbp_4)
      (Variables|/Game/Tools/WaterfallGenerator/WaterfallGeneration_Visual.WaterfallGeneration_Visual_C|SetSplashSize (Variables|IgnoreThis%5BWorkVar%5D|GetMaterialData_SplashSize) _targetvisualbp_4)
      (Variables|/Game/Tools/WaterfallGenerator/WaterfallGeneration_Visual.WaterfallGeneration_Visual_C|SetMaterialList (Variables|IgnoreThis%5BWorkVar%5D|GetMaterialData_BaseMaterialList) (Variables|%5BUser%5DInitialSetting|GetTargetVisualBP)))
    (:then_2
      ;[1 1 1] EffectParam
      (Variables|/Game/Tools/WaterfallGenerator/WaterfallGeneration_Visual.WaterfallGeneration_Visual_C|SetFXCustomFloatParameters (Variables|%5BUser%5DCustomParameterSetting|GetFXCustomFloatParameters) _targetvisualbp_3)
      (Variables|/Game/Tools/WaterfallGenerator/WaterfallGeneration_Visual.WaterfallGeneration_Visual_C|SetFXCustomVecParamters (Variables|%5BUser%5DCustomParameterSetting|GetFXCustomVecParamters) _targetvisualbp_3)
      (Variables|/Game/Tools/WaterfallGenerator/WaterfallGeneration_Visual.WaterfallGeneration_Visual_C|SetBottomPosition _returnvalue _targetvisualbp)
      (Variables|/Game/Tools/WaterfallGenerator/WaterfallGeneration_Visual.WaterfallGeneration_Visual_C|SetTopPosition (Math|Vector|MakeVector (.x _returnvalue) (.y _returnvalue) (.z (Spline|GetLocationatSplinePoint _spline_bevelandfall 2 "World"))) _targetvisualbp)
      (Variables|/Game/Tools/WaterfallGenerator/WaterfallGeneration_Visual.WaterfallGeneration_Visual_C|SetSplashArray (Variables|IgnoreThis%5BWorkVar%5D|GetSplashPosArray) _targetvisualbp))
    (:then_3
      (Variables|/Game/Tools/WaterfallGenerator/WaterfallGeneration_Visual.WaterfallGeneration_Visual_C|SetMaterialCustomFloatParameters (Variables|%5BUser%5DCustomParameterSetting|GetMaterialCustomFloatParameters) _targetvisualbp_2)
      (Variables|/Game/Tools/WaterfallGenerator/WaterfallGeneration_Visual.WaterfallGeneration_Visual_C|SetMaterialCustomVecParameter (Variables|%5BUser%5DCustomParameterSetting|GetMaterialCustomVecParameter) _targetvisualbp_2))
    (:then_4
      (Class|WaterfallGenerationVisual|Refresh (Variables|%5BUser%5DInitialSetting|GetTargetVisualBP)))
    (:then_5
      (Development|PrintString "WaterfallGenerationTool:ExecuteBakeWaterfallData"))))

(fn CreateNewStaticMesh
  (bind _createstaticmeshname (Variables|%5BUser%5DInitialSetting|GetCreateStaticMeshName))
  (bind _returnvalue (Utilities|String|Append "/Game/Mesh/InIslandWaterfall/" _createstaticmeshname))
  (bind _returnvalue_2 (EditorScripting|Asset|DoesAssetExist _returnvalue :!class /Script/EditorScriptingUtilities.EditorAssetLibrary))
  (if _returnvalue_2
    (EditorScripting|MessageDialog|ShowMessageDialog (Utilities|Text|MakeLiteralText "NSLOCTEXT(\"[097BEAC84762376967D75D883370C120]\", \"E09E78164439BA9CDCA9E0B0AB7A132E\", \"Warning\")") (Utilities|Text|MakeLiteralText "NSLOCTEXT(\"[097BEAC84762376967D75D883370C120]\", \"46C605E043875DB66DC2B2A7594F9D47\", \"Target static mesh already exist.Please change name.\")") "Ok" "Ok")
    (else
      (DynamicMeshComponent|GetDynamicMesh (Variables|DynamicMeshActor|GetDynamicMeshComponent))
      (bind _returnvalue_3 (EditorScripting|Asset|DuplicateAsset "/Game/Tools/WaterfallGenerator/Template" _returnvalue :!class /Script/EditorScriptingUtilities.EditorAssetLibrary))
      (Utilities|IsValid _returnvalue_3
        (:Is%20Valid
          (Variables|%5BUser%5DInitialSetting|SetTargetStaticMesh (Utilities|Casting|CastToStaticMesh _returnvalue_3))
          (Development|PrintString "WaterfallGenerationTool:CreateSMAssetSuccess"))
        (:Is%20Not%20Valid
          (Development|PrintString "WaterfallGenerationTool:CreateSMAssetFail")))
      (Variables|%5BUser%5DInitialSetting|SetI'm%20good (and (Utilities|IsValid (Variables|%5BUser%5DInitialSetting|GetTargetStaticMesh) :!class /Script/Engine.KismetSystemLibrary) (Utilities|IsValid (Variables|%5BUser%5DInitialSetting|GetTargetVisualBP) :!class /Script/Engine.KismetSystemLibrary))))))

(fn CreateVisualBP
  (Utilities|IsValid (Variables|%5BUser%5DInitialSetting|GetTargetVisualBP)
    (:Is%20Valid
      (bind _returnvalue (EditorScripting|MessageDialog|ShowMessageDialog (Utilities|Text|MakeLiteralText "NSLOCTEXT(\"[097BEAC84762376967D75D883370C120]\", \"87A188124E23224F8937A19DF5446481\", \"Warning\")") (Utilities|Text|MakeLiteralText "NSLOCTEXT(\"[097BEAC84762376967D75D883370C120]\", \"A23FFC544A5DAA1B114944954BA5510C\", \"Target visual BP existed, do you want to create a new one?This operation will destory old one.\")") "YesNo" "No"))
      (switch Utilities|FlowControl|Switch|SwitchonEAppReturnType _returnvalue
        (:No
          (Development|PrintString "WaterfallGenerationTool:CreateVisualBPFail"))
        (:Yes
          (Actor|DestroyActor (Variables|%5BUser%5DInitialSetting|GetTargetVisualBP))
          (bind _returnvalue_2 (EditorScripting|LevelUtility|SpawnActorfromClass "/Game/Tools/WaterfallGenerator/WaterfallGeneration_Visual.WaterfallGeneration_Visual_C" (.location (Transformation|GetActorTransform)) (.rotation (Transformation|GetActorTransform)) :!class /Script/EditorScriptingUtilities.EditorLevelLibrary))
          (Variables|%5BUser%5DInitialSetting|SetTargetVisualBP _returnvalue_2)
          (Development|PrintString "WaterfallGenerationTool:CreateVisualBPSuccess")))
      (Variables|%5BUser%5DInitialSetting|SetI'm%20good (and (Utilities|IsValid (Variables|%5BUser%5DInitialSetting|GetTargetStaticMesh) :!class /Script/Engine.KismetSystemLibrary) (Utilities|IsValid (Variables|%5BUser%5DInitialSetting|GetTargetVisualBP) :!class /Script/Engine.KismetSystemLibrary))))
    (:Is%20Not%20Valid
      _)))

(fn CurveTrace
  (params (Curve :type Spline%20Component%20Object%20Reference) (Offset :type Vector) (SubDivide :type Integer) (Backward :type Boolean))
  (bind _index (Utilities|FlowControl|ForLoopwithBreak 0 SubDivide)
    (:LoopBody
      (bind (_outhit _returnvalue) (Collision|LineTraceByChannel :Start (+ (Spline|GetLocationatTime Curve (+ (/ (Variables|IgnoreThis%5BWorkVar%5D|GetSplashDetectSafeZone) (Spline|GetSplineLength Curve)) (* (/ (+ (Spline|GetDistanceAlongSplineatSplinePoint Curve 1) (* (Variables|IgnoreThis%5BWorkVar%5D|GetSplashDetectSafeZone) -2.0)) (Spline|GetSplineLength Curve)) (Math|Float|SelectFloat (/ (Math|Conversions|ToFloat(Integer) _index) (Math|Conversions|ToFloat(Integer) (+ SubDivide 1))) (- 1.0 (/ (Math|Conversions|ToFloat(Integer) _index) (Math|Conversions|ToFloat(Integer) (+ SubDivide 1)))) Backward))) "World" true) (Math|Transform|TransformDirection (Transformation|GetActorTransform) Offset)) :End (+ (Spline|GetLocationatTime Curve (+ (/ (Variables|IgnoreThis%5BWorkVar%5D|GetSplashDetectSafeZone) (Spline|GetSplineLength Curve)) (* (/ (+ (Spline|GetDistanceAlongSplineatSplinePoint Curve 1) (* (Variables|IgnoreThis%5BWorkVar%5D|GetSplashDetectSafeZone) -2.0)) (Spline|GetSplineLength Curve)) (Math|Float|SelectFloat (/ (Math|Conversions|ToFloat(Integer) (+ _index 1)) (Math|Conversions|ToFloat(Integer) (+ SubDivide 1))) (- 1.0 (/ (Math|Conversions|ToFloat(Integer) (+ _index 1)) (Math|Conversions|ToFloat(Integer) (+ SubDivide 1)))) Backward))) "World" true) (Math|Transform|TransformDirection (Transformation|GetActorTransform) Offset)) :TraceChannel "TraceTypeQuery1" :bTraceComplex true :DrawDebugType "ForDuration" :bIgnoreSelf true :TraceColor "(R=1.000000,G=0.000000,B=0.000000,A=1.000000)" :TraceHitColor "(R=0.000000,G=1.000000,B=0.000000,A=1.000000)" :DrawTime 5.0))
      (if _returnvalue
        (Utilities|Array|Add (LocalVariables|GetTmpTraceResult) _outhit)))
    (:Completed
      (return (LocalVariables|GetTmpTraceResult)))))

(fn DetectSplashPoint
  (bind _splashdetectcount (Variables|Performance|GetSplashDetectCount))
  (Utilities|Array|Clear (Variables|IgnoreThis%5BWorkVar%5D|GetSplashPosArray))
  (for _index (range (+ _splashdetectcount 1))
    (bind _traceresult (Tool|CurveTrace (Variables|Default|GetSpline_BevelAndFall) (Math|Vector|MakeVector 0.0 (+ (/ (+ (Variables|IgnoreThis%5BWorkVar%5D|GetWidth) 2e+02) -2.0) (* (Math|Conversions|ToFloat(Integer) _index) (/ (+ (Variables|IgnoreThis%5BWorkVar%5D|GetWidth) 2e+02) (Math|Conversions|ToFloat(Integer) _splashdetectcount))))) (Variables|Performance|GetSplashDetectTraceSubDivide)))
    (for _array_element _traceresult
      (Utilities|Array|Add (Variables|IgnoreThis%5BWorkVar%5D|GetSplashPosArray) (.Location (Collision|BreakHitResult _array_element))))))

(fn Generation_MaterialDataCollect
  (Utilities|FlowControl|Sequence
    (:then_0
      (Variables|IgnoreThis%5BWorkVar%5D|SetMaterialData_Width (+ (Variables|IgnoreThis%5BWorkVar%5D|GetWidth) (* (Variables|IgnoreThis%5BWorkVar%5D|GetThickness) 1.57))))
    (:then_1
      (Variables|IgnoreThis%5BWorkVar%5D|SetMaterialData_Height (Spline|GetSplineLength (Variables|Default|GetSpline_BevelAndFall))))
    (:then_2
      (Variables|IgnoreThis%5BWorkVar%5D|SetMaterialData_EdgeOffset (/ (* (Variables|IgnoreThis%5BWorkVar%5D|GetThickness) 1.57) 2.0)))
    (:then_3
      (Variables|IgnoreThis%5BWorkVar%5D|SetMaterialData_BevelLength (Math|Float|SelectFloat (Variables|Cutout|GetRoundFallBevelDistance) (- (Spline|GetSplineLength (Variables|Default|GetSpline_BevelAndFall)) (Spline|GetDistanceAlongSplineatSplinePoint (Variables|Default|GetSpline_BevelAndFall) 1)) (Variables|Cutout|GetRoundFall))))
    (:then_4
      (Variables|IgnoreThis%5BWorkVar%5D|SetMaterialData_ExtendLength (Spline|GetSplineLength (Variables|Default|GetSpline_Extend))))
    (:then_5
      (Variables|IgnoreThis%5BWorkVar%5D|SetMaterialData_SplashSize (Variables|IgnoreThis%5BWorkVar%5D|GetWaveScale)))
    (:then_6
      (Variables|IgnoreThis%5BWorkVar%5D|SetMaterialData_BaseMaterialList (Utilities|Array|MakeArray (Variables|%5BUser%5DInitialSetting|GetOverrideFallMaterial) (Variables|%5BUser%5DInitialSetting|GetOverrideExtendMaterial) (Variables|%5BUser%5DInitialSetting|GetOverrideSplashMaterial))))))

(fn Generation_MeshGen
  ;[1 1 1 1 128] Generate mesh
  (Utilities|FlowControl|Sequence
    (:then_0
      (Utilities|FlowControl|Sequence
        (:then_0
          (bind _returnvalue (DynamicMeshComponent|GetDynamicMesh (Variables|Default|GetDynamicMesh_BevelAndFall)))
          (bind _returnvalue_2 (GeometryScript|Primitives|AppendSimpleSweptPolygon _returnvalue (Utilities|Struct|MakeGeometryScriptPrimitiveOptions "PerFace") (Math|Transform|MakeTransform "0, 0, 0") (Tool|GetSweptPointArray 0.0 (Variables|Performance|GetEdgeSegment) (Variables|Performance|GetCenterSegment)) (Variables|SampleSpline|GetTmpSampleSP_BAF) false false))
          (GeometryScript|UVs|ScaleMeshUVs _returnvalue_2 0 "(X=2.000000,Y=1.000000)"))
        (:then_1
          (bind _returnvalue_3 (DynamicMeshComponent|GetDynamicMesh (Variables|Default|GetDynamicMesh_BevelAndFall_LOD1)))
          (bind _returnvalue_4 (GeometryScript|Primitives|AppendSimpleSweptPolygon _returnvalue_3 (Utilities|Struct|MakeGeometryScriptPrimitiveOptions "PerFace") (Math|Transform|MakeTransform "0, 0, 0") (Tool|GetSweptPointArray 0.0 (Variables|Performance|GetEdgeSegment_LOD1) (Variables|Performance|GetCenterSegment_LOD1)) (Variables|SampleSpline|GetTmpSampleSP_BAF_LOD1) false false))
          (GeometryScript|UVs|ScaleMeshUVs _returnvalue_4 0 "(X=2.000000,Y=1.000000)"))
        (:then_2
          (bind _returnvalue_5 (DynamicMeshComponent|GetDynamicMesh (Variables|Default|GetDynamicMesh_BevelAndFall_LOD2)))
          (bind _returnvalue_6 (GeometryScript|Primitives|AppendSimpleSweptPolygon _returnvalue_5 (Utilities|Struct|MakeGeometryScriptPrimitiveOptions "PerFace") (Math|Transform|MakeTransform "0, 0, 0") (Tool|GetSweptPointArray 0.0 (Variables|Performance|GetEdgeSegment_LOD2) (Variables|Performance|GetCenterSegment_LOD2)) (Variables|SampleSpline|GetTmpSampleSP_BAF_LOD2) false false))
          (GeometryScript|UVs|ScaleMeshUVs _returnvalue_6 0 "(X=2.000000,Y=1.000000)"))))
    (:then_1
      (Utilities|FlowControl|Sequence
        (:then_0
          (bind _returnvalue_7 (DynamicMeshComponent|GetDynamicMesh (Variables|Default|GetDynamicMesh_Extend)))
          (bind _returnvalue_8 (GeometryScript|Primitives|AppendSimpleSweptPolygon _returnvalue_7 (Utilities|Struct|MakeGeometryScriptPrimitiveOptions "PerFace") (Math|Transform|MakeTransform "0, 0, 0") (Tool|GetSweptPointArray -2.0 (Variables|Performance|GetEdgeSegment) (Variables|Performance|GetCenterSegment)) (Variables|SampleSpline|GetTmpSampleSP_Ex) false false 1.0 (Variables|IgnoreThis%5BWorkVar%5D|GetExtend_ScaleEnd)))
          (GeometryScript|UVs|ScaleMeshUVs _returnvalue_8 0 "(X=2.000000,Y=1.000000)"))
        (:then_1
          (bind _returnvalue_9 (DynamicMeshComponent|GetDynamicMesh (Variables|Default|GetDynamicMesh_Extend_LOD1)))
          (bind _returnvalue_10 (GeometryScript|Primitives|AppendSimpleSweptPolygon _returnvalue_9 (Utilities|Struct|MakeGeometryScriptPrimitiveOptions "PerFace") (Math|Transform|MakeTransform "0, 0, 0") (Tool|GetSweptPointArray -2.0 (Variables|Performance|GetEdgeSegment_LOD1) (Variables|Performance|GetCenterSegment_LOD1)) (Variables|SampleSpline|GetTmpSampleSP_Ex_LOD1) false false 1.0 (Variables|IgnoreThis%5BWorkVar%5D|GetExtend_ScaleEnd)))
          (GeometryScript|UVs|ScaleMeshUVs _returnvalue_10 0 "(X=2.000000,Y=1.000000)"))
        (:then_2
          (bind _returnvalue_11 (DynamicMeshComponent|GetDynamicMesh (Variables|Default|GetDynamicMesh_Extend_LOD2)))
          (bind _returnvalue_12 (GeometryScript|Primitives|AppendSimpleSweptPolygon _returnvalue_11 (Utilities|Struct|MakeGeometryScriptPrimitiveOptions "PerFace") (Math|Transform|MakeTransform "0, 0, 0") (Tool|GetSweptPointArray -2.0 (Variables|Performance|GetEdgeSegment_LOD2) (Variables|Performance|GetCenterSegment_LOD2)) (Variables|SampleSpline|GetTmpSampleSP_Ex_LOD2) false false 1.0 (Variables|IgnoreThis%5BWorkVar%5D|GetExtend_ScaleEnd)))
          (GeometryScript|UVs|ScaleMeshUVs _returnvalue_12 0 "(X=2.000000,Y=1.000000)"))))
    (:then_2
      (Utilities|FlowControl|Sequence
        (:then_0
          (bind _returnvalue_13 (DynamicMeshComponent|GetDynamicMesh (Variables|Default|GetDynamicMesh_Wave)))
          (bind _returnvalue_14 (GeometryScript|Primitives|AppendDisc _returnvalue_13 (Utilities|Struct|MakeGeometryScriptPrimitiveOptions "PerFace") (Math|Transform|MakeTransform (Spline|GetLocationatDistanceAlongSpline (Variables|Default|GetSpline_BevelAndFall) 0.0 "Local")) (Variables|IgnoreThis%5BWorkVar%5D|GetWaveScale) (Variables|Performance|GetSplashSegment_AngleStep) (Variables|Performance|GetSplashSegment_SpokeStep)))
          (GeometryScript|Normals|RecomputeNormals _returnvalue_14 (Utilities|Struct|MakeGeometryScriptCalculateNormalsOptions true)))
        (:then_1
          (bind _returnvalue_15 (DynamicMeshComponent|GetDynamicMesh (Variables|Default|GetDynamicMesh_Wave_LOD1)))
          (bind _returnvalue_16 (GeometryScript|Primitives|AppendDisc _returnvalue_15 (Utilities|Struct|MakeGeometryScriptPrimitiveOptions "PerFace") (Math|Transform|MakeTransform (Spline|GetLocationatDistanceAlongSpline (Variables|Default|GetSpline_BevelAndFall) 0.0 "Local")) (Variables|IgnoreThis%5BWorkVar%5D|GetWaveScale) (Variables|Performance|GetSplashSegment_AngleStep_LOD1) (Variables|Performance|GetSplashSegment_SpokeStep_LOD1)))
          (GeometryScript|Normals|RecomputeNormals _returnvalue_16 (Utilities|Struct|MakeGeometryScriptCalculateNormalsOptions true)))
        (:then_2
          (bind _returnvalue_17 (DynamicMeshComponent|GetDynamicMesh (Variables|Default|GetDynamicMesh_Wave_LOD2)))
          (bind _returnvalue_18 (GeometryScript|Primitives|AppendDisc _returnvalue_17 (Utilities|Struct|MakeGeometryScriptPrimitiveOptions "PerFace") (Math|Transform|MakeTransform (Spline|GetLocationatDistanceAlongSpline (Variables|Default|GetSpline_BevelAndFall) 0.0 "Local")) (Variables|IgnoreThis%5BWorkVar%5D|GetWaveScale) (Variables|Performance|GetSplashSegment_AngleStep_LOD2) (Variables|Performance|GetSplashSegment_SpokeStep_LOD2)))
          (GeometryScript|Normals|RecomputeNormals _returnvalue_18 (Utilities|Struct|MakeGeometryScriptCalculateNormalsOptions true)))))))

(fn Generation_MeshMerge
  ;[1 1 1 1 128] Merge mesh
  (Utilities|FlowControl|Sequence
    (:then_0
      (Utilities|FlowControl|Sequence
        (:then_0
          (bind _returnvalue (DynamicMeshComponent|GetDynamicMesh (Variables|Default|GetDynamicMesh_BevelAndFall)))
          (bind _returnvalue_2 (GeometryScript|Materials|EnableMaterialIDs _returnvalue))
          (bind (_triangleidlist _bhastriangleidgaps _returnvalue_3) (GeometryScript|MeshQueries|GetAllTriangleIDs _returnvalue_2))
          (bind _returnvalue_4 (GeometryScript|Materials|SetMaterialIDOnTriangles _returnvalue_3 _triangleidlist)))
        (:then_1
          (bind _returnvalue_5 (DynamicMeshComponent|GetDynamicMesh (Variables|Default|GetDynamicMesh_Extend)))
          (bind _returnvalue_6 (GeometryScript|Materials|EnableMaterialIDs _returnvalue_5))
          (bind (_triangleidlist_2 _bhastriangleidgaps_2 _returnvalue_7) (GeometryScript|MeshQueries|GetAllTriangleIDs _returnvalue_6))
          (bind _returnvalue_8 (GeometryScript|Materials|SetMaterialIDOnTriangles _returnvalue_7 _triangleidlist_2 1)))
        (:then_2
          (bind _returnvalue_9 (DynamicMeshComponent|GetDynamicMesh (Variables|Default|GetDynamicMesh_Wave)))
          (bind _returnvalue_10 (GeometryScript|Materials|EnableMaterialIDs _returnvalue_9))
          (bind (_triangleidlist_3 _bhastriangleidgaps_3 _returnvalue_11) (GeometryScript|MeshQueries|GetAllTriangleIDs _returnvalue_10))
          (bind _returnvalue_12 (GeometryScript|Materials|SetMaterialIDOnTriangles _returnvalue_11 _triangleidlist_3 2)))
        (:then_4
          (bind _returnvalue_13 (DynamicMeshComponent|GetDynamicMesh (Variables|DynamicMeshActor|GetDynamicMeshComponent)))
          (bind _returnvalue_14 (GeometryScript|Materials|EnableMaterialIDs _returnvalue_13))
          (bind _returnvalue_15 (GeometryScript|MeshEdits|AppendMesh :TargetMesh _returnvalue_14 :AppendMesh _returnvalue_4 :bDeferChangeNotifications false :AppendOptions "(CombineMode=EnableAllMatching)"))
          (bind _returnvalue_16 (GeometryScript|MeshEdits|AppendMesh :TargetMesh _returnvalue_15 :AppendMesh _returnvalue_8 :bDeferChangeNotifications false :AppendOptions "(CombineMode=EnableAllMatching)"))
          (GeometryScript|MeshEdits|AppendMesh :TargetMesh _returnvalue_16 :AppendMesh _returnvalue_12 :bDeferChangeNotifications false :AppendOptions "(CombineMode=EnableAllMatching)"))))
    (:then_1
      (Utilities|FlowControl|Sequence
        (:then_0
          (bind _returnvalue_17 (DynamicMeshComponent|GetDynamicMesh (Variables|Default|GetDynamicMesh_BevelAndFall_LOD1)))
          (bind _returnvalue_18 (GeometryScript|Materials|EnableMaterialIDs _returnvalue_17))
          (bind (_triangleidlist_4 _bhastriangleidgaps_4 _returnvalue_19) (GeometryScript|MeshQueries|GetAllTriangleIDs _returnvalue_18))
          (bind _returnvalue_20 (GeometryScript|Materials|SetMaterialIDOnTriangles _returnvalue_19 _triangleidlist_4)))
        (:then_1
          (bind _returnvalue_21 (DynamicMeshComponent|GetDynamicMesh (Variables|Default|GetDynamicMesh_Extend_LOD1)))
          (bind _returnvalue_22 (GeometryScript|Materials|EnableMaterialIDs _returnvalue_21))
          (bind (_triangleidlist_5 _bhastriangleidgaps_5 _returnvalue_23) (GeometryScript|MeshQueries|GetAllTriangleIDs _returnvalue_22))
          (bind _returnvalue_24 (GeometryScript|Materials|SetMaterialIDOnTriangles _returnvalue_23 _triangleidlist_5 1)))
        (:then_2
          (bind _returnvalue_25 (DynamicMeshComponent|GetDynamicMesh (Variables|Default|GetDynamicMesh_Wave_LOD1)))
          (bind _returnvalue_26 (GeometryScript|Materials|EnableMaterialIDs _returnvalue_25))
          (bind (_triangleidlist_6 _bhastriangleidgaps_6 _returnvalue_27) (GeometryScript|MeshQueries|GetAllTriangleIDs _returnvalue_26))
          (bind _returnvalue_28 (GeometryScript|Materials|SetMaterialIDOnTriangles _returnvalue_27 _triangleidlist_6 2)))
        (:then_4
          (bind _returnvalue_29 (DynamicMeshComponent|GetDynamicMesh (Variables|Default|GetDynamicMesh_Merged_LOD1)))
          (bind _returnvalue_30 (GeometryScript|Materials|EnableMaterialIDs _returnvalue_29))
          (bind _returnvalue_31 (GeometryScript|MeshEdits|AppendMesh :TargetMesh _returnvalue_30 :AppendMesh _returnvalue_20 :bDeferChangeNotifications false :AppendOptions "(CombineMode=EnableAllMatching)"))
          (bind _returnvalue_32 (GeometryScript|MeshEdits|AppendMesh :TargetMesh _returnvalue_31 :AppendMesh _returnvalue_24 :bDeferChangeNotifications false :AppendOptions "(CombineMode=EnableAllMatching)"))
          (GeometryScript|MeshEdits|AppendMesh :TargetMesh _returnvalue_32 :AppendMesh _returnvalue_28 :bDeferChangeNotifications false :AppendOptions "(CombineMode=EnableAllMatching)"))))
    (:then_2
      (Utilities|FlowControl|Sequence
        (:then_0
          (bind _returnvalue_33 (DynamicMeshComponent|GetDynamicMesh (Variables|Default|GetDynamicMesh_BevelAndFall_LOD2)))
          (bind _returnvalue_34 (GeometryScript|Materials|EnableMaterialIDs _returnvalue_33))
          (bind (_triangleidlist_7 _bhastriangleidgaps_7 _returnvalue_35) (GeometryScript|MeshQueries|GetAllTriangleIDs _returnvalue_34))
          (bind _returnvalue_36 (GeometryScript|Materials|SetMaterialIDOnTriangles _returnvalue_35 _triangleidlist_7)))
        (:then_1
          (bind _returnvalue_37 (DynamicMeshComponent|GetDynamicMesh (Variables|Default|GetDynamicMesh_Extend_LOD2)))
          (bind _returnvalue_38 (GeometryScript|Materials|EnableMaterialIDs _returnvalue_37))
          (bind (_triangleidlist_8 _bhastriangleidgaps_8 _returnvalue_39) (GeometryScript|MeshQueries|GetAllTriangleIDs _returnvalue_38))
          (bind _returnvalue_40 (GeometryScript|Materials|SetMaterialIDOnTriangles _returnvalue_39 _triangleidlist_8 1)))
        (:then_2
          (bind _returnvalue_41 (DynamicMeshComponent|GetDynamicMesh (Variables|Default|GetDynamicMesh_Wave_LOD2)))
          (bind _returnvalue_42 (GeometryScript|Materials|EnableMaterialIDs _returnvalue_41))
          (bind (_triangleidlist_9 _bhastriangleidgaps_9 _returnvalue_43) (GeometryScript|MeshQueries|GetAllTriangleIDs _returnvalue_42))
          (bind _returnvalue_44 (GeometryScript|Materials|SetMaterialIDOnTriangles _returnvalue_43 _triangleidlist_9 2)))
        (:then_4
          (bind _returnvalue_45 (DynamicMeshComponent|GetDynamicMesh (Variables|Default|GetDynamicMesh_Merged_LOD2)))
          (bind _returnvalue_46 (GeometryScript|Materials|EnableMaterialIDs _returnvalue_45))
          (bind _returnvalue_47 (GeometryScript|MeshEdits|AppendMesh :TargetMesh _returnvalue_46 :AppendMesh _returnvalue_36 :bDeferChangeNotifications false :AppendOptions "(CombineMode=EnableAllMatching)"))
          (bind _returnvalue_48 (GeometryScript|MeshEdits|AppendMesh :TargetMesh _returnvalue_47 :AppendMesh _returnvalue_40 :bDeferChangeNotifications false :AppendOptions "(CombineMode=EnableAllMatching)"))
          (GeometryScript|MeshEdits|AppendMesh :TargetMesh _returnvalue_48 :AppendMesh _returnvalue_44 :bDeferChangeNotifications false :AppendOptions "(CombineMode=EnableAllMatching)"))))))

(fn Generation_PreviewMaterialSetting
  (bind _materialdata_basemateriallist (Variables|IgnoreThis%5BWorkVar%5D|GetMaterialData_BaseMaterialList))
  (bind _output (Utilities|Array|Get(acopy) _materialdata_basemateriallist 2))
  (bind _materialdata_basemateriallist_2 (Variables|IgnoreThis%5BWorkVar%5D|GetMaterialData_BaseMaterialList))
  (bind _output_2 (Utilities|Array|Get(acopy) _materialdata_basemateriallist_2))
  (bind _materialdata_basemateriallist_3 (Variables|IgnoreThis%5BWorkVar%5D|GetMaterialData_BaseMaterialList))
  (bind _output_3 (Utilities|Array|Get(acopy) _materialdata_basemateriallist_3 1))
  (bind _array (Utilities|Array|MakeArray "/Game/Tools/WaterfallGenerator/Art/M_Waterfall_GeoDebug_Wire.M_Waterfall_GeoDebug_Wire" "/Game/Tools/WaterfallGenerator/Art/M_Waterfall_GeoDebug_Wire.M_Waterfall_GeoDebug_Wire" "/Game/Tools/WaterfallGenerator/Art/M_Waterfall_GeoDebug_Wire.M_Waterfall_GeoDebug_Wire"))
  (bind _auto_splashmaterial (Variables|IgnoreThis|Get%5BAuto%5DSplashMaterial))
  (bind _auto_extendmaterial (Variables|IgnoreThis|Get%5BAuto%5DExtendMaterial))
  (bind _auto_waterfallmaterial (Variables|IgnoreThis|Get%5BAuto%5DWaterFallMaterial))
  (bind _array_2 (Utilities|Array|MakeArray _auto_waterfallmaterial _auto_extendmaterial _auto_splashmaterial))
  (bind _array_3 (Utilities|Array|MakeArray "/Game/Tools/WaterfallGenerator/Art/M_Waterfall_GeoDebug_UV.M_Waterfall_GeoDebug_UV" "/Game/Tools/WaterfallGenerator/Art/M_Waterfall_GeoDebug_UV.M_Waterfall_GeoDebug_UV" "/Game/Tools/WaterfallGenerator/Art/M_Waterfall_GeoDebug_UV.M_Waterfall_GeoDebug_UV"))
  (bind _array_4 (Utilities|Array|MakeArray "/Game/Tools/WaterfallGenerator/Art/M_Waterfall_geoDebug_FaceOverlap.M_Waterfall_geoDebug_FaceOverlap" "/Game/Tools/WaterfallGenerator/Art/M_Waterfall_geoDebug_FaceOverlap.M_Waterfall_geoDebug_FaceOverlap" "/Game/Tools/WaterfallGenerator/Art/M_Waterfall_geoDebug_FaceOverlap.M_Waterfall_geoDebug_FaceOverlap"))
  (bind _debugtype (Variables|%5BUser%5DInitialSetting|GetDebugType))
  (bind _returnvalue (Utilities|Select :Index _debugtype :Option%200 _array :Option%201 _array_3 :Option%202 _array_4 :Option%203 _array_2))
  (bind _materialcustomfloatparameters (Variables|%5BUser%5DCustomParameterSetting|GetMaterialCustomFloatParameters))
  (bind _auto_waterfallmaterial_2 (Variables|IgnoreThis|Get%5BAuto%5DWaterFallMaterial))
  (bind _materialcustomvecparameter (Variables|%5BUser%5DCustomParameterSetting|GetMaterialCustomVecParameter))
  ;[1 1 1 1 128] Material things
  (Utilities|FlowControl|Sequence
    (:then_0
      ;;[1 1 1] DM
      (bind _returnvalue_2 (Rendering|Material|GetBaseMaterial (Variables|IgnoreThis|Get%5BAuto%5DWaterFallMaterial)))
      (if (not (== _returnvalue_2 _output_2))
        (bind _returnvalue_3 (Rendering|Material|CreateDynamicMaterialInstance _output_2 "None" "None" :!class /Script/Engine.KismetMaterialLibrary))
        (Variables|IgnoreThis|Set%5BAuto%5DWaterFallMaterial _returnvalue_3)))
    (:then_1
      (bind _returnvalue_4 (Rendering|Material|GetBaseMaterial (Variables|IgnoreThis|Get%5BAuto%5DExtendMaterial)))
      (if (not (== _returnvalue_4 _output_3))
        (bind _returnvalue_5 (Rendering|Material|CreateDynamicMaterialInstance _output_3 "None" "None" :!class /Script/Engine.KismetMaterialLibrary))
        (Variables|IgnoreThis|Set%5BAuto%5DExtendMaterial _returnvalue_5)))
    (:then_2
      (bind _returnvalue_6 (Rendering|Material|GetBaseMaterial (Variables|IgnoreThis|Get%5BAuto%5DSplashMaterial)))
      (if (not (== _returnvalue_6 _output))
        (bind _returnvalue_7 (Rendering|Material|CreateDynamicMaterialInstance _output "None" "None" :!class /Script/Engine.KismetMaterialLibrary))
        (Variables|IgnoreThis|Set%5BAuto%5DSplashMaterial _returnvalue_7)))
    (:then_3
      ;;[1 1 1] BasicMaterialParameterSetting
      (Rendering|Material|SetScalarParameterValue _auto_waterfallmaterial_2 "BPInput_Width" (Variables|IgnoreThis%5BWorkVar%5D|GetMaterialData_Width) :!class /Script/Engine.MaterialInstanceDynamic)
      (Rendering|Material|SetScalarParameterValue _auto_waterfallmaterial_2 "BPInput_Height" (Variables|IgnoreThis%5BWorkVar%5D|GetMaterialData_Height) :!class /Script/Engine.MaterialInstanceDynamic)
      (Rendering|Material|SetScalarParameterValue _auto_waterfallmaterial_2 "BPInput_EdgeOffset" (Variables|IgnoreThis%5BWorkVar%5D|GetMaterialData_EdgeOffset) :!class /Script/Engine.MaterialInstanceDynamic)
      (Rendering|Material|SetScalarParameterValue _auto_waterfallmaterial_2 "BPInput_BevelLength" (Variables|IgnoreThis%5BWorkVar%5D|GetMaterialData_BevelLength) :!class /Script/Engine.MaterialInstanceDynamic)
      (Rendering|Material|SetScalarParameterValue (Variables|IgnoreThis|Get%5BAuto%5DExtendMaterial) "BPInput_ExtendLength" (Variables|IgnoreThis%5BWorkVar%5D|GetMaterialData_ExtendLength) :!class /Script/Engine.MaterialInstanceDynamic)
      (Rendering|Material|SetScalarParameterValue (Variables|IgnoreThis|Get%5BAuto%5DSplashMaterial) "BPInput_SplashSize" (Variables|IgnoreThis%5BWorkVar%5D|GetMaterialData_SplashSize) :!class /Script/Engine.MaterialInstanceDynamic))
    (:then_4
      ;;[1 1 1] CustomMaterialParameterSetting
      (Utilities|FlowControl|Sequence
        (:then_0
          (bind _keys (Utilities|Map|Keys _materialcustomfloatparameters))
          (bind _values (Utilities|Map|Values _materialcustomfloatparameters))
          (for (_array_element _array_index) _keys
            (Rendering|Material|SetScalarParameterValue (Variables|IgnoreThis|Get%5BAuto%5DExtendMaterial) _array_element (Utilities|Array|Get(aref) _values _array_index) :!class /Script/Engine.MaterialInstanceDynamic)))
        (:then_1
          (bind _keys_2 (Utilities|Map|Keys _materialcustomvecparameter))
          (bind _values_2 (Utilities|Map|Values _materialcustomvecparameter))
          (for (_array_element_2 _array_index_2) _keys_2
            (Rendering|Material|SetVectorParameterValue (Variables|IgnoreThis|Get%5BAuto%5DExtendMaterial) _array_element_2 (Utilities|Array|Get(aref) _values_2 _array_index_2) :!class /Script/Engine.MaterialInstanceDynamic)))))
    (:then_5
      (GeometryScript|Scene|SetComponentMaterialList (Variables|DynamicMeshActor|GetDynamicMeshComponent) _returnvalue)
      (GeometryScript|Scene|SetComponentMaterialList (Variables|Default|GetDynamicMesh_Merged_LOD1) _returnvalue)
      (GeometryScript|Scene|SetComponentMaterialList (Variables|Default|GetDynamicMesh_Merged_LOD2) _returnvalue))))

(fn Generation_SampleSpline
  ;[1 1 1 1 128] Collect Generation data
  (Utilities|FlowControl|Sequence
    (:then_0
      (Utilities|FlowControl|Sequence
        (:then_0
          (CallFunction|GenerationSubFuncSampleBAF (Variables|Performance|GetSampleCount_Fall) (Variables|SampleSpline|GetTmpSampleSP_BAF) (Variables|Performance|GetSampleCount_Bevel)))
        (:then_1
          (CallFunction|GenerationSubFuncSampleBAF (Variables|Performance|GetSampleCount_Fall_LOD1) (Variables|SampleSpline|GetTmpSampleSP_BAF_LOD1) (Variables|Performance|GetSampleCount_Bevel_LOD1)))
        (:then_2
          (CallFunction|GenerationSubFuncSampleBAF (Variables|Performance|GetSampleCount_Fall_LOD2) (Variables|SampleSpline|GetTmpSampleSP_BAF_LOD2) (Variables|Performance|GetSampleCount_Bevel_LOD2)))))
    (:then_1
      (Utilities|FlowControl|Sequence
        (:then_0
          (CallFunction|GenerationSubFuncSampleEX (Variables|Performance|GetSampleCount_Bevel) (Variables|Performance|GetSampleCount_Extend) (Variables|SampleSpline|GetTmpSampleSP_Ex)))
        (:then_1
          (CallFunction|GenerationSubFuncSampleEX (Variables|Performance|GetSampleCount_Bevel) (Variables|Performance|GetSampleCount_Extend) (Variables|SampleSpline|GetTmpSampleSP_Ex_LOD1)))
        (:then_2
          (CallFunction|GenerationSubFuncSampleEX (Variables|Performance|GetSampleCount_Bevel) (Variables|Performance|GetSampleCount_Extend) (Variables|SampleSpline|GetTmpSampleSP_Ex_LOD2)))))))

(fn Generation_SplineGen
  (bind _bottom (Variables|Widget|Get◉Bottom))
  (bind _x (.x _bottom))
  (bind _bevel_v (Variables|Widget|Get➤Bevel_V))
  (bind _y (.y _bevel_v))
  (bind _returnvalue (Math|Float|Absolute(Float) _y))
  (bind _samplecount_extend (Variables|Performance|GetSampleCount_Extend))
  (bind _returnvalue_2 (Math|Conversions|ToFloat(Integer) _samplecount_extend))
  (bind _returnvalue_3 (* _returnvalue_2 2.0))
  (bind _spline_extend (Variables|Default|GetSpline_Extend))
  (bind _returnvalue_4 (Spline|GetSplineLength _spline_extend))
  (bind _returnvalue_5 (/ _returnvalue_4 _returnvalue_3))
  (bind _returnvalue_6 (Math|Vector|MakeVector _returnvalue_5))
  (bind _top (Variables|Widget|Get◉Top))
  (bind _z (.z _top))
  (bind _returnvalue_7 (- _z _returnvalue))
  (bind _spline_extend_2 (Variables|Default|GetSpline_Extend))
  (bind _spline_bevelandfall (Variables|Default|GetSpline_BevelAndFall))
  (bind _bevel_h (Variables|Widget|Get◉Bevel_H))
  (bind _spline_extend_3 (Variables|Default|GetSpline_Extend))
  (bind _bevel_h_2 (Variables|Widget|Get◉Bevel_H))
  (bind _spline_extend_4 (Variables|Default|GetSpline_Extend))
  (bind _spline_extend_5 (Variables|Default|GetSpline_Extend))
  (bind _spline_extend_6 (Variables|Default|GetSpline_Extend))
  (bind _spline_bevelandfall_2 (Variables|Default|GetSpline_BevelAndFall))
  (bind _spline_bevelandfall_3 (Variables|Default|GetSpline_BevelAndFall))
  ;[1 1 1 1 128] SplineGen
  (Spline|ClearSplinePoints (Variables|Default|GetSpline_BevelAndFall) true)
  (Utilities|FlowControl|Sequence
    (:then_0
      (Spline|AddSplinePoint _spline_bevelandfall_2 (Variables|Widget|Get◉Bottom) "Local" false)
      (Spline|SetTangentsatSplinePoint _spline_bevelandfall_2 0 (Math|Vector|MakeVector 0.0) (Math|Vector|MakeVector 0.0 0.0 (Math|Float|SelectFloat (* (* (- (.z (Variables|Widget|Get◉Top)) (.z (Variables|Widget|Get◉Bottom))) 2.0) (+ (/ (* (Math|Float|Sign(Float) (.y (Variables|Widget|Get➤Bevel_V))) (Math|Float|Max(Float) (- (Math|Float|Absolute(Float) (.y (Variables|Widget|Get➤Bevel_V))) 1e+02))) 3e+02) 1.0)) 0.0 (Variables|Cutout|GetRoundFall))) "Local" false))
    (:then_1
      (Spline|AddSplinePoint _spline_bevelandfall (Math|Vector|MakeVector (Math|Float|Lerp _x (.x _top) (/ _returnvalue_7 (- _z (.z _bottom)))) (.y _top) _returnvalue_7) "Local" false)
      (Spline|SetTangentsatSplinePoint _spline_bevelandfall 1 (Math|Vector|MakeVector 0.0) (Math|Vector|MakeVector (* _x -0.5) 0.0 (* (- _z _returnvalue_7) 2.0)) "Local" false))
    (:then_2
      (Spline|AddSplinePoint _spline_bevelandfall_3 _bevel_h "Local" false)
      (Spline|SetTangentsatSplinePoint _spline_bevelandfall_3 2 (Math|Vector|MakeVector (* (.x _bevel_h) (Variables|IgnoreThis%5BWorkVar%5D|GetBevel_H_ArriveTangentScale))) (Math|Vector|MakeVector (Variables|IgnoreThis%5BWorkVar%5D|GetBevel_H_LeaveTangent)) "Local" false))
    (:then_3
      (if (Variables|Cutout|GetRoundFall)
        (Spline|RemoveSplinePoint (Variables|Default|GetSpline_BevelAndFall) 1 true)
        (else
          (Spline|UpdateSpline (Variables|Default|GetSpline_BevelAndFall)))))
    (:then_5
      ;;[1 1 1] SafeZome
      (Spline|AddSplinePoint _spline_extend_6 (+ (Variables|Widget|Get◉Bevel_H) 0) "Local" false)
      (Spline|SetTangentsatSplinePoint _spline_extend_6 0 (Math|Vector|MakeVector 0.0) (Math|Vector|MakeVector 0.0) "Local" false)
      (Spline|AddSplinePoint _spline_extend_4 (+ (Variables|Widget|Get◉Bevel_H) 0) "Local" false)
      (Spline|SetTangentsatSplinePoint _spline_extend_4 1 (Math|Vector|MakeVector 0.0) (Math|Vector|MakeVector 0.0) "Local" false))
    (:then_6
      (Spline|AddSplinePoint _spline_extend_5 (+ (Variables|Widget|Get◉Bevel_H) "1.000000,0,0") "Local" false)
      (Spline|SetTangentsatSplinePoint _spline_extend_5 2 (Math|Vector|MakeVector 0.0) (Math|Vector|MakeVector (Variables|IgnoreThis%5BWorkVar%5D|GetBevel_H_LeaveTangent)) "Local" false))
    (:then_7
      (Spline|AddSplinePoint _spline_extend_2 (Variables|Widget|Get◉Extend) "Local" false)
      (Spline|SetTangentsatSplinePoint _spline_extend_2 3 (Math|Vector|MakeVector 0.0) (Math|Vector|MakeVector 0.0) "Local" false))
    (:then_8
      (Spline|UpdateSpline (Variables|Default|GetSpline_Extend))
      ;;[1 1 1] SafeZone
      (Spline|SetLocationatSplinePoint _spline_extend_3 1 (Utilities|Operators|Add _returnvalue_6 _bevel_h_2 "0.010000,0,0") "Local" true)
      (Spline|SetLocationatSplinePoint _spline_extend_3 2 (Utilities|Operators|Add _bevel_h_2 (* _returnvalue_6 "2.000000,2.000000,2.000000") "0.010000,0,0") "Local" true))))

(fn Generation_SubFunc_SampleBAF
  (params (SampleCount_Fall :type Integer) (TargetArray :type Array%20of%20Vectors) (SampleCount_Bevel :type Integer))
  (Utilities|FlowControl|Sequence
    (:then_0
      ;[1 1 1] Comment
      (for _index (range (+ SampleCount_Fall 1))
        (Utilities|Array|Add (LocalVariables|GetTmpArray) (Spline|GetLocationatTime (Variables|Default|GetSpline_BevelAndFall) (* (/ (Math|Conversions|ToFloat(Integer) _index) (Math|Conversions|ToFloat(Integer) SampleCount_Fall)) (/ (Spline|GetDistanceAlongSplineatSplinePoint (Variables|Default|GetSpline_BevelAndFall) 1) (Spline|GetSplineLength (Variables|Default|GetSpline_BevelAndFall)))) "Local" true))))
    (:then_1
      ;[1 1 1] Comment
      (for _index_2 (range 1 (+ SampleCount_Bevel 1))
        (Utilities|Array|Add (LocalVariables|GetTmpArray) (Spline|GetLocationatTime (Variables|Default|GetSpline_BevelAndFall) (+ (* (/ (Math|Conversions|ToFloat(Integer) _index_2) (Math|Conversions|ToFloat(Integer) SampleCount_Bevel)) (- 1.0 (/ (Spline|GetDistanceAlongSplineatSplinePoint (Variables|Default|GetSpline_BevelAndFall) 1) (Spline|GetSplineLength (Variables|Default|GetSpline_BevelAndFall))))) (/ (Spline|GetDistanceAlongSplineatSplinePoint (Variables|Default|GetSpline_BevelAndFall) 1) (Spline|GetSplineLength (Variables|Default|GetSpline_BevelAndFall)))) "Local" true))))
    (:then_2
      ;[1 1 1] Fix
      (Utilities|Array|Add (LocalVariables|GetTmpArray) (+ (Spline|GetLocationatSplinePoint (Variables|Default|GetSpline_Extend) 0 "Local") "0.010000,0,0")))
    (:then_3
      (Variables|SetBy-RefVar TargetArray (LocalVariables|GetTmpArray)))))

(fn Generation_SubFunc_SampleEX
  (params (SampleCount_Bevel :type Integer) (SampleCount_Extend :type Integer) (NewParam :type Array%20of%20Vectors))
  (Utilities|FlowControl|Sequence
    (:then_0
      ;[1 1 1] Comment
      (for _index (range 1 SampleCount_Bevel)
        (Utilities|Array|Add (LocalVariables|GetTmpArray) (Spline|GetLocationatTime (Variables|Default|GetSpline_BevelAndFall) (+ (* (/ (Math|Conversions|ToFloat(Integer) _index) (Math|Conversions|ToFloat(Integer) SampleCount_Bevel)) (- 1.0 (/ (Spline|GetDistanceAlongSplineatSplinePoint (Variables|Default|GetSpline_BevelAndFall) 1) (Spline|GetSplineLength (Variables|Default|GetSpline_BevelAndFall))))) (/ (Spline|GetDistanceAlongSplineatSplinePoint (Variables|Default|GetSpline_BevelAndFall) 1) (Spline|GetSplineLength (Variables|Default|GetSpline_BevelAndFall)))) "Local" true))))
    (:then_1
      ;[1 1 1] Comment
      (for _index_2 (range (+ SampleCount_Extend 1))
        (Utilities|Array|Add (LocalVariables|GetTmpArray) (Spline|GetLocationatTime (Variables|Default|GetSpline_Extend) (/ (Math|Conversions|ToFloat(Integer) _index_2) (Math|Conversions|ToFloat(Integer) SampleCount_Extend)) "Local" true))))
    (:then_2
      (Variables|SetBy-RefVar NewParam (LocalVariables|GetTmpArray)))))

(fn GetSweptPointArray
  (params (Offset :type Float%20%28double-precision%29) (EdgeSeg :type Integer) (CenterSeg :type Integer))
  (bind _width (Variables|IgnoreThis%5BWorkVar%5D|GetWidth))
  (bind _returnvalue (+ _width (Variables|GetOffset)))
  (bind _returnvalue_2 (/ _returnvalue 2.0))
  (bind _offset (Variables|GetOffset))
  (Utilities|FlowControl|Sequence
    (:then_0
      (Utilities|FlowControl|Sequence
        (:then_1
          ;[1 1 1] Prepare
          (Utilities|FlowControl|Sequence
            (:then_0
              (LocalVariables|SetLeftO (Math|Vector2D|MakeVector2D 0.0 (* _returnvalue_2 -1.0)))
              (LocalVariables|SetRightO (Math|Vector2D|MakeVector2D 0.0 _returnvalue_2)))
            (:then_1
              (LocalVariables|SetR (/ (+ _offset (Variables|IgnoreThis%5BWorkVar%5D|GetThickness)) 2.0)))
            (:then_2
              (LocalVariables|SetAlpha (/ 9e+01 (Math|Conversions|ToFloat(Integer) (+ (Variables|GetEdgeSeg) 1)))))
            (:then_3
              (LocalVariables|SetCentSegLength (/ (+ (Variables|IgnoreThis%5BWorkVar%5D|GetWidth) _offset) (Math|Conversions|ToFloat(Integer) (+ (Variables|GetCenterSeg) 1)))))))
        (:then_2
          ;[1 1 1] -    -    -
          ; /               \
          ; T                 *
          ; \               /
          ; -    -    -
          (Utilities|Array|Add (LocalVariables|GetPointArray) (- (LocalVariables|GetLeftO) (Math|Vector2D|MakeVector2D 0.0 (LocalVariables|GetR)))))
        (:then_3
          ;[1 1 1] -    -    -
          ; /               \
          ; *                 *
          ; T               /
          ; -    -    -
          (bind _index (Utilities|FlowControl|ForLoopwithBreak 1 (Variables|GetEdgeSeg))
            (:LoopBody
              (Utilities|Array|Add (LocalVariables|GetPointArray) (+ (LocalVariables|GetLeftO) (* (Math|Vector2D|MakeVector2D (* (Math|Trig|Cos(Degrees) (- 9e+01 (* (LocalVariables|GetAlpha) (Math|Conversions|ToFloat(Integer) _index)))) 1.0) (* (Math|Trig|Sin(Degrees) (- 9e+01 (* (LocalVariables|GetAlpha) (Math|Conversions|ToFloat(Integer) _index)))) -1.0)) (Math|Vector2D|MakeVector2D (LocalVariables|GetR) (LocalVariables|GetR))))))))
        (:then_4
          ;[1 1 1] -    -    -
          ; /               \
          ; *                 *
          ; \               /
          ; T    -    -
          (Utilities|Array|Add (LocalVariables|GetPointArray) (+ (LocalVariables|GetLeftO) (Math|Vector2D|MakeVector2D (LocalVariables|GetR)))))
        (:then_5
          ;[1 1 1] -    -    -
          ; /               \
          ; *                 *
          ; \               /
          ; -    T    -
          (bind _index_2 (Utilities|FlowControl|ForLoopwithBreak 1 (Variables|GetCenterSeg))
            (:LoopBody
              (Utilities|Array|Add (LocalVariables|GetPointArray) (Utilities|Operators|Add (LocalVariables|GetLeftO) (Math|Vector2D|MakeVector2D (LocalVariables|GetR)) (Math|Vector2D|MakeVector2D 0.0 (* (LocalVariables|GetCentSegLength) (Math|Conversions|ToFloat(Integer) _index_2))))))))
        (:then_6
          ;[1 1 1] -    -    -
          ; /               \
          ; *                 *
          ; \               /
          ; -    -    T
          (Utilities|Array|Add (LocalVariables|GetPointArray) (+ (LocalVariables|GetRightO) (Math|Vector2D|MakeVector2D (LocalVariables|GetR)))))
        (:then_7
          ;[1 1 1] -    -    -
          ; /               \
          ; *                 *
          ; \               T
          ; -    -    -
          (bind _index_3 (Utilities|FlowControl|ForLoopwithBreak 1 (Variables|GetEdgeSeg))
            (:LoopBody
              (Utilities|Array|Add (LocalVariables|GetPointArray) (+ (LocalVariables|GetRightO) (* (Math|Vector2D|MakeVector2D (* (Math|Trig|Cos(Degrees) (* (LocalVariables|GetAlpha) (Math|Conversions|ToFloat(Integer) _index_3))) 1.0) (* (Math|Trig|Sin(Degrees) (* (LocalVariables|GetAlpha) (Math|Conversions|ToFloat(Integer) _index_3))) 1.0)) (Math|Vector2D|MakeVector2D (LocalVariables|GetR) (LocalVariables|GetR))))))))
        (:then_8
          ;[1 1 1] -    -    -
          ; /               \
          ; *                 T
          ; \               /
          ; -    -    -
          (Utilities|Array|Add (LocalVariables|GetPointArray) (+ (LocalVariables|GetRightO) (Math|Vector2D|MakeVector2D 0.0 (LocalVariables|GetR)))))
        (:then_9
          ;[1 1 1] -    -    -
          ; /               T
          ; *                 *
          ; \               /
          ; -    -    -
          (bind _index_4 (Utilities|FlowControl|ForLoopwithBreak 1 (Variables|GetEdgeSeg))
            (:LoopBody
              (Utilities|Array|Add (LocalVariables|GetPointArray) (+ (LocalVariables|GetRightO) (* (Math|Vector2D|MakeVector2D (* (Math|Trig|Cos(Degrees) (- 9e+01 (* (LocalVariables|GetAlpha) (Math|Conversions|ToFloat(Integer) _index_4)))) -1.0) (* (Math|Trig|Sin(Degrees) (- 9e+01 (* (LocalVariables|GetAlpha) (Math|Conversions|ToFloat(Integer) _index_4)))) 1.0)) (Math|Vector2D|MakeVector2D (LocalVariables|GetR) (LocalVariables|GetR))))))))
        (:then_10
          ;[1 1 1] -    -    T
          ; /               \
          ; *                 *
          ; \               /
          ; -    -    -
          (Utilities|Array|Add (LocalVariables|GetPointArray) (- (LocalVariables|GetRightO) (Math|Vector2D|MakeVector2D (LocalVariables|GetR)))))
        (:then_11
          ;[1 1 1] -    T    -
          ; /               \
          ; *                 *
          ; \               /
          ; -    -    -
          (bind _index_5 (Utilities|FlowControl|ForLoopwithBreak 1 (Variables|GetCenterSeg))
            (:LoopBody
              (Utilities|Array|Add (LocalVariables|GetPointArray) (Utilities|Operators|Add (LocalVariables|GetRightO) (Math|Vector2D|MakeVector2D (* (LocalVariables|GetR) -1.0)) (Math|Vector2D|MakeVector2D 0.0 (* (* (LocalVariables|GetCentSegLength) (Math|Conversions|ToFloat(Integer) _index_5)) -1.0)))))))
        (:then_12
          ;[1 1 1] T    -    -
          ; /               \
          ; *                 *
          ; \               /
          ; -    -    -
          (Utilities|Array|Add (LocalVariables|GetPointArray) (- (LocalVariables|GetLeftO) (Math|Vector2D|MakeVector2D (LocalVariables|GetR)))))
        (:then_13
          ;[1 1 1] -    -    -
          ; /               \
          ; *                 *
          ; \               T
          ; -    -    -
          (bind _index_6 (Utilities|FlowControl|ForLoopwithBreak 1 (Variables|GetEdgeSeg))
            (:LoopBody
              (Utilities|Array|Add (LocalVariables|GetPointArray) (+ (LocalVariables|GetLeftO) (* (Math|Vector2D|MakeVector2D (* (Math|Trig|Cos(Degrees) (* (LocalVariables|GetAlpha) (Math|Conversions|ToFloat(Integer) _index_6))) -1.0) (* (Math|Trig|Sin(Degrees) (* (LocalVariables|GetAlpha) (Math|Conversions|ToFloat(Integer) _index_6))) -1.0)) (Math|Vector2D|MakeVector2D (LocalVariables|GetR) (LocalVariables|GetR))))))))))
    (:then_1
      (return (LocalVariables|GetPointArray)))))

(fn OpenDocument
  (Utilities|Platform|LaunchURL "https://docs.google.com/document/d/1qvA07-39mHjOz19oNX0GfhYxXEDGRBJoTwGmP7ldoYU/edit?usp=sharing"))

(fn RefreshPriviewFX)

(fn ConstructionScript
  (bind _targetvisualbp (Variables|%5BUser%5DInitialSetting|GetTargetVisualBP))
  (bind _forzen (Variables|Default|GetForzen))
  (Utilities|FlowControl|Sequence
    (:then_0
      ;[1 1 1] Check
      (Variables|%5BUser%5DInitialSetting|SetI'm%20good (and (Utilities|IsValid (Variables|%5BUser%5DInitialSetting|GetTargetStaticMesh) :!class /Script/Engine.KismetSystemLibrary) (Utilities|IsValid (Variables|%5BUser%5DInitialSetting|GetTargetVisualBP) :!class /Script/Engine.KismetSystemLibrary))))
    (:then_1
      ;[1 1 1] SetupTickHelper
      (Variables|/Game/Tools/WaterfallGenerator/WaterfallGenerator_TickHelper.WaterfallGenerator_TickHelper_C|SetMyParent self (Variables|Default|GetTickHelper)))
    (:then_2
      ;[1 1 1] MoveVisualActorAndRefreshIt
      (if (Variables|%5BUser%5DInitialSetting|GetI'm%20good)
        (Variables|/Game/Tools/WaterfallGenerator/WaterfallGeneration_Visual.WaterfallGeneration_Visual_C|SetTargetSM (Variables|%5BUser%5DInitialSetting|GetTargetStaticMesh) _targetvisualbp)
        (Transformation|SetActorTransform _targetvisualbp (Transformation|GetActorTransform) false false)
        (Class|WaterfallGenerationVisual|Refresh _targetvisualbp)))
    (:then_4
      ;[1 1 1] DebugTextVisbility
      (if (Variables|DynamicMeshActor|GetbFrozen)
        (Rendering|SetVisibility _forzen true false)
        (else
          (Rendering|SetVisibility _forzen false false))))))

(fn UserInterface_Limit3DWidget
  (bind _bevel_h_leavetangentw (Variables|Widget|Get➤Bevel_H_LeaveTangentW))
  (bind _y (.y _bevel_h_leavetangentw))
  (bind _extend_scaleendw (Variables|Widget|Get➤Extend_ScaleEndW))
  (bind _y_2 (.y _extend_scaleendw))
  (bind _wavescalew (Variables|Widget|Get➤WaveScaleW))
  (bind _x (.x _wavescalew))
  (bind _wdithpoint (Variables|Widget|Get◉WdithPoint))
  (bind _y_3 (.y _wdithpoint))
  (bind _thicknessw (Variables|Widget|Get➤ThicknessW))
  (bind _x_2 (.x _thicknessw))
  (bind _returnvalue (Math|Float|Min(Float) _x_2 -1.0))
  (bind _returnvalue_2 (Math|Float|Max(Float) _y))
  (bind _width (Variables|IgnoreThis%5BWorkVar%5D|GetWidth))
  (bind _returnvalue_3 (/ _width 2.0))
  (bind _thickness (Variables|IgnoreThis%5BWorkVar%5D|GetThickness))
  (bind _returnvalue_4 (/ _thickness 2.0))
  (bind _returnvalue_5 (+ _returnvalue_4 _returnvalue_3))
  (bind _returnvalue_6 (* _returnvalue_5 -1.0))
  (bind _returnvalue_7 (Math|Float|Min(Float) _x _returnvalue_6))
  ;[1 1 1 1 128] Constraint
  (Utilities|FlowControl|Sequence
    (:then_0
      ;;[1 1 1] Top
      (Variables|Widget|Set◉Top (Math|Vector|MakeVector 0.0 0.0 (.z (Variables|Widget|Get◉Top)))))
    (:then_1
      ;;[1 1 1] Bottom
      (Variables|Widget|Set◉Bottom (Math|Vector|MakeVector (.x (Variables|Widget|Get◉Bottom)) 0.0 (.z (Variables|Widget|Get◉Bottom)))))
    (:then_2
      ;;[1 1 1] Extend
      (Variables|Widget|Set◉Extend (Math|Vector|MakeVector (.x (Variables|Widget|Get◉Extend)) (.y (Variables|Widget|Get◉Extend)) (.z (Variables|Widget|Get◉Top)))))
    (:then_3
      ;;[1 1 1] BevelH
      (Variables|Widget|Set◉Bevel_H (Math|Vector|MakeVector (Math|Float|Max(Float) (.x (Variables|Widget|Get◉Bevel_H))) 0.0 (.z (Variables|Widget|Get◉Top)))))
    (:then_4
      ;;[1 1 1] BevelV
      (Variables|Widget|Set➤Bevel_V (Math|Vector|MakeVector 0.0 (.y (Variables|Widget|Get➤Bevel_V)) (.z (Variables|Widget|Get◉Top)))))
    (:then_5
      ;;[1 1 1] Width
      (Variables|Widget|Set◉WdithPoint (Math|Vector|MakeVector (.x (Variables|Widget|Get◉Bottom)) (Math|Float|Max(Float) _y_3 1.0) (.z (Variables|Widget|Get◉Bottom))))
      (Variables|IgnoreThis%5BWorkVar%5D|SetWidth (* _y_3 2.0)))
    (:then_6
      ;;[1 1 1] Thickness
      (Variables|Widget|Set➤ThicknessW (Math|Vector|MakeVector _returnvalue 0.0 (* (+ (.z (Variables|Widget|Get◉Top)) (.z (Variables|Widget|Get◉Bottom))) 0.5)))
      (Variables|IgnoreThis%5BWorkVar%5D|SetThickness (/ (Math|Float|Absolute(Float) _returnvalue) 2.0)))
    (:then_7
      ;;[1 1 1] ScaleEnd
      (Variables|Widget|Set➤Extend_ScaleEndW (Math|Vector|MakeVector (.x (Variables|Widget|Get◉Bevel_H)) _y_2 (+ (.z (Variables|Widget|Get◉Bevel_H)) 4e+02)))
      (Variables|IgnoreThis%5BWorkVar%5D|SetExtend_ScaleEnd (+ (/ (* (Math|Float|Max(Float) (- (Math|Float|Absolute(Float) _y_2) 5e+01)) (Math|Float|Sign(Float) _y_2)) 1e+02) 1.0)))
    (:then_8
      ;;[1 1 1] LeaveTangent
      (Variables|Widget|Set➤Bevel_H_LeaveTangentW (Math|Vector|MakeVector (.x (Variables|Widget|Get◉Bevel_H)) _returnvalue_2 (+ (.z (Variables|Widget|Get◉Bevel_H)) 2e+02)))
      (Variables|IgnoreThis%5BWorkVar%5D|SetBevel_H_LeaveTangent (* _returnvalue_2 2.0)))
    (:then_9
      (Variables|Widget|Set➤WaveScaleW (Math|Vector|MakeVector _returnvalue_7 (.y (Variables|Widget|Get◉Bottom)) (.z (Variables|Widget|Get◉Bottom))))
      (Variables|IgnoreThis%5BWorkVar%5D|SetWaveScale (Math|Float|Absolute(Float) _returnvalue_7)))))
