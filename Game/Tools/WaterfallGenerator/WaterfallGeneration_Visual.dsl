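; Waterfall actor for the waterfall generator tool that renders the fall and bounces away players who touch it.
; Construction applies the mesh, materials, shape and custom parameters, feeds Niagara, and sizes collision boxes over the fall and splash base.
; BeginPlay re-pushes Niagara parameters.
; Box overlaps launch eligible pawns backward from the mesh's forward and upward with a gameplay effect, fall-damage immunity and force feedback; non-pawn overlaps eject overlapping vehicles' occupants.
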(event EventBeginPlay
  (CallFunction|UpdateNiagaraParam))

(event Collision|EventActorBeginOverlap
  (params (OtherActor :type Actor%20Object%20Reference)))

(event EventTick
  (params (DeltaSeconds :type Float%20%28single-precision%29)))

(event OnComponentBeginOverlap(Bounce)
  (params (OverlappedComponent :type Primitive%20Component%20Object%20Reference) (OtherActor :type Actor%20Object%20Reference) (OtherComp :type Primitive%20Component%20Object%20Reference) (OtherBodyIndex :type Integer) (bFromSweep :type Boolean) (SweepResult :type Hit%20Result%20Structure%20%28by%20ref%29))
  (bind _gp_currentactor (Variables|Gameplay|GetGP_CurrentActor))
  (bind _returnvalue (Pawn|GetController _gp_currentactor))
  (bind _gp_currentactor_2 (Variables|Gameplay|GetGP_CurrentActor))
  (bind _asfort_pawn (Utilities|Casting|CastToFortPawn OtherActor)
    (:then
      (Variables|Gameplay|SetGP_CurrentActor _asfort_pawn)
      (if (not (or (GameplayTags|HasMatchingGameplayTag (Ability|GetAbilitySystemComponent (Variables|Gameplay|GetGP_CurrentActor)) "(TagName=\"Gameplay.Action.Player.Launched\")") (Variables|/Script/FortniteGame.FortPawn|GetbIsDBNO (Variables|Gameplay|GetGP_CurrentActor))))
        (bind _asfort_player_pawn (Utilities|Casting|CastToFortPlayerPawn (Variables|Gameplay|GetGP_CurrentActor))
          (:then
            (RemoteControlPawn|ForceDetachfromRemoteControlledPawn _asfort_player_pawn))
          (:CastFailed
            _))
        (Character|LaunchCharacter (Variables|Gameplay|GetGP_CurrentActor) (+ (* (* (Transformation|GetForwardVector (Variables|Default|GetWaterfallMesh)) "-1.000000,-1.000000,-1.000000") "500.000000,500.000000,500.000000") "0.000000,0.000000,500.000000") true true)
        (GameplayEffects|ApplyGameplayEffectToSelf :self (Ability|GetAbilitySystemComponent _gp_currentactor_2) :GameplayEffectClass (Variables|Gameplay|GetGE_PlayerLaunched) :Level 0.0)
        (Variables|/Script/FortniteGame.FortPawn|SetbIgnoreNextFallingDamage true _gp_currentactor_2)
        (Utilities|IsValid _returnvalue
          (:Is%20Valid
            (bind _asfort_player_controller (Utilities|Casting|CastToFortPlayerController _returnvalue))
            (Game|Feedback|LocalPlayForceFeedback _asfort_player_controller (Variables|Gameplay|GetLaunchForceFeedback) false "TireLaunch")))))
    (:CastFailed
      ;[1 1 1] Eject players from vehicles
      (for _array_element (Collision|GetOverlappingActors (Variables|Default|GetBounce) "/Script/FortniteGame.FortAthenaVehicle" :!class /Script/Engine.PrimitiveComponent)
        (Vehicle|EjectAllPlayers(Message) _array_element))
      (for _array_element_2 (Collision|GetOverlappingActors (Variables|Default|GetBounce1) "/Script/FortniteGame.FortAthenaVehicle" :!class /Script/Engine.PrimitiveComponent)
        (Vehicle|EjectAllPlayers(Message) _array_element_2)))))

(event OnComponentBeginOverlap(Bounce1)
  (params (OverlappedComponent :type Primitive%20Component%20Object%20Reference) (OtherActor :type Actor%20Object%20Reference) (OtherComp :type Primitive%20Component%20Object%20Reference) (OtherBodyIndex :type Integer) (bFromSweep :type Boolean) (SweepResult :type Hit%20Result%20Structure%20%28by%20ref%29))
  (bind _gp_currentactor (Variables|Gameplay|GetGP_CurrentActor))
  (bind _returnvalue (Pawn|GetController _gp_currentactor))
  (bind _gp_currentactor_2 (Variables|Gameplay|GetGP_CurrentActor))
  (bind _asfort_pawn (Utilities|Casting|CastToFortPawn OtherActor)
    (:then
      (Variables|Gameplay|SetGP_CurrentActor _asfort_pawn)
      (if (not (or (GameplayTags|HasMatchingGameplayTag (Ability|GetAbilitySystemComponent (Variables|Gameplay|GetGP_CurrentActor)) "(TagName=\"Gameplay.Action.Player.Launched\")") (Variables|/Script/FortniteGame.FortPawn|GetbIsDBNO (Variables|Gameplay|GetGP_CurrentActor))))
        (bind _asfort_player_pawn (Utilities|Casting|CastToFortPlayerPawn (Variables|Gameplay|GetGP_CurrentActor))
          (:then
            (RemoteControlPawn|ForceDetachfromRemoteControlledPawn _asfort_player_pawn))
          (:CastFailed
            _))
        (Character|LaunchCharacter (Variables|Gameplay|GetGP_CurrentActor) (+ (* (* (Transformation|GetForwardVector (Variables|Default|GetWaterfallMesh)) "-1.000000,-1.000000,-1.000000") "500.000000,500.000000,500.000000") "0.000000,0.000000,500.000000") true true)
        (GameplayEffects|ApplyGameplayEffectToSelf :self (Ability|GetAbilitySystemComponent _gp_currentactor_2) :GameplayEffectClass (Variables|Gameplay|GetGE_PlayerLaunched) :Level 0.0)
        (Variables|/Script/FortniteGame.FortPawn|SetbIgnoreNextFallingDamage true _gp_currentactor_2)
        (Utilities|IsValid _returnvalue
          (:Is%20Valid
            (bind _asfort_player_controller (Utilities|Casting|CastToFortPlayerController _returnvalue))
            (Game|Feedback|LocalPlayForceFeedback _asfort_player_controller (Variables|Gameplay|GetLaunchForceFeedback) false "TireLaunch")))))
    (:CastFailed
      ;[1 1 1] Eject players from vehicles
      (for _array_element (Collision|GetOverlappingActors (Variables|Default|GetBounce) "/Script/FortniteGame.FortAthenaVehicle" :!class /Script/Engine.PrimitiveComponent)
        (Vehicle|EjectAllPlayers(Message) _array_element))
      (for _array_element_2 (Collision|GetOverlappingActors (Variables|Default|GetBounce1) "/Script/FortniteGame.FortAthenaVehicle" :!class /Script/Engine.PrimitiveComponent)
        (Vehicle|EjectAllPlayers(Message) _array_element_2)))))

(fn GameplayThingsRefresh
  (bind _bottomposition (Variables|DoNotModifyHere|GetBottomPosition))
  (bind _topposition (Variables|DoNotModifyHere|GetTopPosition))
  (bind _returnvalue (- _topposition _bottomposition))
  (bind _returnvalue_2 (/ _returnvalue "2.000000,2.000000,2.000000"))
  (bind _bounce (Variables|Default|GetBounce))
  (bind _bounce1 (Variables|Default|GetBounce1))
  (Utilities|FlowControl|Sequence
    (:then_0
      (Transformation|SetWorldLocation _bounce (+ _bottomposition _returnvalue_2) false false)
      (Components|Box|SetBoxExtent _bounce (Math|Vector|MakeVector (* (Variables|DoNotModifyHere|GetThickness) 1.5) (* (Variables|DoNotModifyHere|GetWidth) 0.5) (+ (.z _returnvalue_2) 5e+01)) true))
    (:then_1
      (Transformation|SetWorldLocation _bounce1 _bottomposition false false)
      (Components|Box|SetBoxExtent _bounce1 (Math|Vector|MakeVector (* (Variables|DoNotModifyHere|GetSplashSize) 0.7) (* (Variables|DoNotModifyHere|GetSplashSize) 0.7) 1.5e+02) true))))

(fn Refresh
  (bind _waterfallmesh (Variables|Default|GetWaterfallMesh))
  (bind _materialcustomfloatparameters (Variables|DoNotModifyHere|GetMaterialCustomFloatParameters))
  (bind _materialcustomvecparameter (Variables|DoNotModifyHere|GetMaterialCustomVecParameter))
  (Utilities|IsValid (Variables|DoNotModifyHere|GetTargetSM)
    (:Is%20Valid
      (Components|StaticMesh|SetStaticMesh (Variables|Default|GetWaterfallMesh) (Variables|DoNotModifyHere|GetTargetSM))
      (Utilities|FlowControl|Sequence
        (:then_1
          ;[1 1 1] MaterialSetting
          (for (_array_element _array_index) (Variables|DoNotModifyHere|GetMaterialList)
            (Rendering|Material|SetMaterial (Variables|Default|GetWaterfallMesh) _array_index _array_element)))
        (:then_2
          ;[1 1 1] BaseMaterialParamterSet
          (Rendering|Material|SetScalarParameterValueonMaterials _waterfallmesh "BPInput_Width" (Variables|DoNotModifyHere|GetWidth))
          (Rendering|Material|SetScalarParameterValueonMaterials _waterfallmesh "BPInput_Height" (Variables|DoNotModifyHere|GetHeight))
          (Rendering|Material|SetScalarParameterValueonMaterials _waterfallmesh "BPInput_EdgeOffset" (Variables|DoNotModifyHere|GetEdgeOffset))
          (Rendering|Material|SetScalarParameterValueonMaterials _waterfallmesh "BPInput_BevelLength" (Variables|DoNotModifyHere|GetBevelLength))
          (Rendering|Material|SetScalarParameterValueonMaterials _waterfallmesh "BPInput_ExtendLength" (Variables|DoNotModifyHere|GetExtendLength))
          (Rendering|Material|SetScalarParameterValueonMaterials _waterfallmesh "BPInput_SplashSize" (Variables|DoNotModifyHere|GetSplashSize))
          (Rendering|Material|SetVectorParameterValueonMaterials _waterfallmesh "BPInput_BPPosition" (Transformation|GetActorLocation)))
        (:then_3
          ;[1 1 1] CustomMaterialParameterSetting
          (Utilities|FlowControl|Sequence
            (:then_0
              (bind _keys (Utilities|Map|Keys _materialcustomfloatparameters))
              (bind _values (Utilities|Map|Values _materialcustomfloatparameters))
              (for (_array_element_2 _array_index_2) _keys
                (Rendering|Material|SetScalarParameterValueonMaterials (Variables|Default|GetWaterfallMesh) _array_element_2 (Utilities|Array|Get(aref) _values _array_index_2))))
            (:then_1
              (bind _keys_2 (Utilities|Map|Keys _materialcustomvecparameter))
              (bind _values_2 (Utilities|Map|Values _materialcustomvecparameter))
              (for (_array_element_3 _array_index_3) _keys_2
                (Rendering|Material|SetVectorParameterValueonMaterials (Variables|Default|GetWaterfallMesh) _array_element_3 (Math|Conversions|ToVector(LinearColor) (Utilities|Array|Get(aref) _values_2 _array_index_3)))))))
        (:then_4
          ;[1 1 1] NiagaraUpdate
          (CallFunction|UpdateNiagaraParam))))))

(fn UpdateNiagaraParam
  (bind _fxcustomfloatparameters (Variables|DoNotModifyHere|GetFXCustomFloatParameters))
  (bind _fxcustomvecparamters (Variables|DoNotModifyHere|GetFXCustomVecParamters))
  (bind _niagara (Variables|Default|GetNiagara))
  (Utilities|IsValid _niagara
    (:Is%20Valid
      (Utilities|FlowControl|Sequence
        (:then_0
          (Niagara|SetNiagaraVariable(Position) _niagara "BottomPosition" (Variables|DoNotModifyHere|GetBottomPosition))
          (Niagara|SetNiagaraVariable(Position) _niagara "TopPosition" (Variables|DoNotModifyHere|GetTopPosition))
          (Niagara|SetNiagaraVariable(Float) _niagara "FallWidth" (+ (Variables|DoNotModifyHere|GetWidth) (* (Variables|DoNotModifyHere|GetEdgeOffset) 2.0)))
          (Niagara|SetNiagaraVariable(Float) _niagara "FallHeight" (Variables|DoNotModifyHere|GetHeight))
          (Niagara|NiagaraSetPositionArray _niagara "SpawnSplashPositionArray" (Variables|DoNotModifyHere|GetSplashArray)))
        (:then_1
          (bind _keys (Utilities|Map|Keys _fxcustomfloatparameters))
          (bind _values (Utilities|Map|Values _fxcustomfloatparameters))
          (for (_array_element _array_index) _keys
            (Niagara|SetNiagaraVariable(Float) (Variables|Default|GetNiagara) _array_element (Utilities|Array|Get(aref) _values _array_index))))
        (:then_2
          (bind _keys_2 (Utilities|Map|Keys _fxcustomvecparamters))
          (bind _values_2 (Utilities|Map|Values _fxcustomvecparamters))
          (for (_array_element_2 _array_index_2) _keys_2
            (Niagara|SetNiagaraVariable(Vector3) (Variables|Default|GetNiagara) _array_element_2 (Utilities|Array|Get(aref) _values_2 _array_index_2))))))))

(fn ConstructionScript
  (Utilities|FlowControl|Sequence
    (:then_0
      (CallFunction|Refresh))
    (:then_1
      (CallFunction|GameplayThingsRefresh))))
