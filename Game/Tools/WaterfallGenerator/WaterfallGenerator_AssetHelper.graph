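# Editor utility Blueprint for the WaterfallGenerator tool, with helper functions that adjust static mesh asset settings.
# DisableStaticMeshDF sets bGenerateMeshDistanceField to false through SetEditorProperty; its save node is not wired into execution.
# LODSettings sets per-quality minimum LODs for Low and Medium, beside an unwired placeholder Python script node.
# CopySMToNewFolder is an empty stub, and the event graph has no logic.
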EditorUtilityBlueprint /Game/Tools/WaterfallGenerator/WaterfallGenerator_AssetHelper
graphs [
 CopySMToNewFolder {nodes [CopySMToNewFolder() {is=title K2Node_FunctionEntry}]}
 DisableStaticMeshDF {nodes [
   n1 = DisableStaticMeshDF!() -> (then, TargetSM:Static Mesh Object Reference) {is=title K2Node_FunctionEntry}
   > SetEditorProperty(Object=@n1, PropertyName=bGenerateMeshDistanceField, PropertyValue=AND Boolean(A=false, B=false), ChangeNotifyMode=Always)
   SaveLoadedAsset(AssetToSave=@n1, bOnlyIfIsDirty=true)
  ]}
 EventGraph {nodes [
   Event Begin Play()
   Event Tick!() -> (OutputDelegate, then, DeltaSeconds:float)
  ]}
 LODSettings {nodes [
   n1 = LODSettings!() -> (then, TargetSm:Static Mesh Object Reference) {is=title K2Node_FunctionEntry}
   > SetMinLODForQualityLevels!(
    execute
    self:Static Mesh Object Reference=@n1
    QualityLevelMinimumLODs:Map of EPerQualityLevels Enums to Integers=Make Map!(Key 0:EPerQualityLevels Enum=Low, Value 0:int=1, Key 1:EPerQualityLevels Enum=Medium, Value 1:int=1) -> (Map:Map of EPerQualityLevels Enums to Integers) {is=title K2Node_MakeMap}
    Default:int=-1
   ) -> (then)
   Execute Python Script!(execute, PythonScript="print('Hello World!')", TargetSM:Wildcard) -> (then, ReturnValue) {inputs [TargetSM]}
  ]}
]
variables []
