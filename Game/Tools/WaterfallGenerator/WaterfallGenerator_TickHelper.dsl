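; Helper Blueprint for the waterfall generator tool that relays bake and FX-detection fix calls to its owning parent.
; BakeWaterfallData and FXDetect_Fix each validate the MyParent reference, then send the matching WaterfallGeneratorTickHelperInterface message to it.
; The interface messages keep the bake and fix logic in the parent, while BeginPlay and Tick here are empty despite the TickHelper name.
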(event EventBeginPlay)

(event EventTick
  (params (DeltaSeconds :type Float%20%28single-precision%29)))

(fn BakeWaterfallData
  (bind _myparent (Variables|Default|GetMyParent))
  (Utilities|IsValid _myparent
    (:Is%20Valid
      (Class|WaterfallGeneratorTickHelperInterface|BakeWaterfallData(Message) _myparent))))

(fn FXDetect_Fix
  (bind _myparent (Variables|Default|GetMyParent))
  (Utilities|IsValid _myparent
    (:Is%20Valid
      (Class|WaterfallGeneratorTickHelperInterface|FXDetectFix(Message) _myparent))))
