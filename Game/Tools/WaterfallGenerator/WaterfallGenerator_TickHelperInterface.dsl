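; Blueprint interface used by the waterfall generator tool, judging by its name as a tick helper.
; It declares two functions, BakeWaterfallData and FXDetect_Fix, with no parameters or logic.
; Implementers supply the baking of waterfall data and the FX detection fix behavior.
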(fn BakeWaterfallData)

(fn FXDetect_Fix)
